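# Material function that turns incoming character hair attributes into flyaway strands with subsurface shading, overriding color, normal, specular, roughness and depth offset.
# A spline-thicken function, widened by vertex green, combines with root-to-tip normal offsets to shape strands.
# Tiled strand textures and painted color add detail, with fresnel brightening and simulated transmission scaled by a cosmetics emissive collection parameter.
# Temporal dither drives pixel depth offset to blend strands, while ShadowReplace removes their shadows.
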MaterialFunction /Game/Global/Materials/Character_Parent_2023/Hair_Parent_2023/MaterialFunctions/MF_FlyAwayHair_2023
graph {nodes [
  n1 = GetMaterialAttributes(None=FunctionInput() {inputName=MaterialInput, inputType=FunctionInput_MaterialAttributes}) {attributeGetTypes [
    #1
    #2
    #3
    #4
   ]}
  n2 = Power(Base=NamedRerouteUsage()) {constExponent=1.5}
  n3 = LinearInterpolate(B=ScalarParameter() {defaultValue=0.5, group=FlyAwayHair, parameterName=FurBlend AlphaThreshold}, Alpha=@n2)
  n4 = Constant() {r=1}
  n5 = ScalarParameter() {defaultValue=1.5, group=FlyAwayHair, parameterName=FurBlendOpacityMask}
  n6 = SetMaterialAttributes!(
   MaterialAttributes=@n1.MaterialAttributes
   WorldPositionOffset=NamedRerouteUsage()
   Normal=NamedRerouteUsage()
   PixelDepthOffset=Multiply(
    A=MaterialFunctionCall(Alpha Threshold=@n3, Random=@n4) {materialFunction=&DitherTemporalAA}.Result
    B=LinearInterpolate(B=@n5, Alpha=@n2)
   )
   BaseColor=NamedRerouteUsage()
   SubsurfaceColor=Constant()
   Opacity=Constant() {r=1}
   OpacityMask=ShadowReplace(Default=Constant() {r=1}, Shadow=Constant())
   Specular=NamedRerouteUsage()
   Roughness=NamedRerouteUsage()
   ShadingModel=ShadingModel() {shadingModel=MSM_Subsurface}
   Metallic=ScalarParameter() {defaultValue=0.2, group=FlyAwayHair, parameterName=FlyHair Metallic}
   Tangent
   Anisotropy
   EmissiveColor=NamedRerouteUsage()
  ) {attributeSetTypes [
    #5
    #1
    #6
    #2
    #7
    #8
    #9
    #3
    #4
    #10
    #11
    #12
    #13
    #14
   ]}
  FunctionOutput(None=@n6)
  n7 = NamedRerouteUsage()
  FunctionOutput(None=@n7) {outputName=OutPut paint hair texture}
  n8 = VertexColor()
  n9 = LinearInterpolate(
   A=ScalarParameter() {defaultValue=0.2, group=FlyAwayHair, parameterName=SplineThicken_WidthBase_Min, sortPriority=1}
   B=ScalarParameter() {defaultValue=0.15, group=FlyAwayHair, parameterName=SplineThicken_WidthBase_Middle, sortPriority=1}
   Alpha=Saturate(None=@n8.G)
  )
  n10 = If(
   A=@n8.G
   "A > B"=ScalarParameter() {defaultValue=0.25, group=FlyAwayHair, parameterName=SplineThicken_WidthBase_ Max, sortPriority=1}
   "A == B"=@n9
   "A < B"=@n9
  ) {constB=0.7}
  n11 = MaterialFunctionCall!(
   widthTip=Multiply(A=ScalarParameter() {defaultValue=0.1, group=FlyAwayHair, parameterName=SplineThicken_Width tip, sortPriority=1}, B=@n10)
   Flatten=ScalarParameter() {group=FlyAwayHair, parameterName=SplineThicken_Flatten, sortPriority=1}
   widthBase=Multiply(A=@n10, B=ScalarParameter() {defaultValue=1, group=FlyAwayHair, parameterName=thickness multiple})
  ) {materialFunction=&/MF_splineThicken_FlyawayHair}
  n12 = MaterialFunctionCall(
   Normal=@n11.normal
   Flatness=ScalarParameter() {defaultValue=0.75, group=FlyAwayHair, parameterName=SplineThicken_NormalFlatness, sortPriority=1}
  ) {materialFunction=&FlattenNormal}
  n13 = NamedRerouteUsage()
  NamedRerouteDeclaration(None=Multiply(
    A=MaterialFunctionCall(BaseNormal=@n12.Result, AdditionalNormal=@n13) {materialFunction=&BlendAngleCorrectedNormals}.Result
    B=TwoSidedSign()
   )) {name=FinalNormal, nodeColor=<rgba 0.9453 0 1 1>}
  n14 = TextureSampleParameter2D() {
   constCoordinate=1
   group=FN_Hair
   parameterName=Base_Diffuse
   samplerSource=SSM_Wrap_WorldGroupSettings
   samplerType=SAMPLERTYPE_Masks
   sortPriority=6
   texture=&T_BlackMask
  }
  NamedRerouteDeclaration(None=FunctionOutput(None=NamedRerouteDeclaration(None=@n14.RGB) {name=InPutBaseColor, nodeColor=<rgba 0.2422 1 0 1>}) {outputName=OutPutBasecolor}) {name=InputPaintHairTexture, nodeColor=<rgba 1 0.6328 0 1>}
  n15 = NamedRerouteUsage()
  NamedRerouteDeclaration(None=LinearInterpolate(
    A=ScalarParameter() {group=FlyAwayHair, parameterName=Flyaway_SpecularMIn}
    B=ScalarParameter() {defaultValue=1, group=FlyAwayHair, parameterName=Flyaway_SpecularMax}
    Alpha=Power(Base=@n15)
   )) {name=FinalSpecular, nodeColor=<rgba 0.3125 0 1 1>}
  n16 = MaterialFunctionCall(
   UVs=TextureCoordinate()
   Texture Scale=MaterialFunctionCall(X=ScalarParameter() {defaultValue=2, group=FlyAwayHair, parameterName=FlyAwayHair_Tiling, sortPriority=2}, Y=Constant() {r=1}) {materialFunction=&MakeFloat2}.Result
  ) {materialFunction=&ScaleUVsByCenter}
  n17 = TextureSampleParameter2D(UVs=@n16.UVs) {
   group=FlyAwayHair
   parameterName=FlyAwayHairTexture
   samplerSource=SSM_Wrap_WorldGroupSettings
   texture=/Game/Global/Materials/Character_Parent_2023/Hair_Parent_2023/Texture/T_FlyAwayHair_D
  }
  NamedRerouteDeclaration(None=LinearInterpolate(
    A=ScalarParameter() {defaultValue=0.4, group=FlyAwayHair, parameterName=Flyaway_RoughnessMIn}
    B=ScalarParameter() {defaultValue=0.6, group=FlyAwayHair, parameterName=Flyaway_RoughnessMax}
    Alpha=OneMinus(None=@n17.R)
   ) {constA=0.8, constB=0.1}) {name=Final_Roughness, nodeColor=<rgba 0.0078 1 0 1>}
  n18 = ScalarParameter() {defaultValue=1, group=FlyAwayHair, parameterName=FlyAwayHair_RootBrightness, sortPriority=3}
  n19 = ScalarParameter() {defaultValue=3.5, group=FlyAwayHair, parameterName=FlyAwayHair_Tip Brightness, sortPriority=3}
  n20 = Power(Base=NamedRerouteUsage(), Exp=Constant() {r=5})
  n21 = NamedRerouteUsage()
  n22 = NamedRerouteUsage()
  n23 = Saturate(None=Multiply(A=NamedRerouteUsage()))
  n24 = Multiply(
   A=Multiply(A=LinearInterpolate(A=@n18, B=@n19, Alpha=@n20) {constA=1, constB=20}, B=LinearInterpolate(A=@n21, B=@n22, Alpha=@n23))
   B=NamedRerouteUsage()
  )
  n25 = Multiply(A=Multiply(A=@n24), B=ScalarParameter() {defaultValue=8, group=FN_Hair, parameterName=Fresnel_Brightness_Multiple, sortPriority=1})
  NamedRerouteDeclaration(None=Saturate(None=Add(A=@n25, B=@n24))) {name=FinalBaseColor, nodeColor=<rgba 0 0.8125 1 1>}
  n26 = VectorParameter() {group=FN_Hair, parameterName=Paint_Hair_Color_Darkness, sortPriority=1, defaultValue=<rgba 0.5 0.5 0.5 1>}
  n27 = VectorParameter() {group=FN_Hair, parameterName=Paint_Hair_Color_Brightness, sortPriority=1, defaultValue=<rgba 1 1 1 1>}
  NamedRerouteDeclaration(None=LinearInterpolate(A=Multiply(A=@n7, B=@n26.RGB) {constB=0.1}, B=Multiply(A=@n7, B=@n27.RGB) {constB=1.5}, Alpha=NamedRerouteUsage())) {name=OutPutBaseColor, nodeColor=<rgba 0.2891 1 0 1>}
  n28 = CollectionParameter() {
   collection=&FortniteMaterialParameters
   desc="CAPTURE EMISSIVE-GLOW GATE  —  FORT-1106795\r\n\r\nWHY: Capture's controlled/dark lighting makes this glow blow out on dark hair && fur.This lets them dim/kill it for a shot without touching the authored look.\r\n\r\nDRIVEN BY: console cvar  Fort.Cosmetics.EmissiveMultiplier"
   parameterName=CosmeticsEmissiveMultiplier
  }
  n29 = NamedRerouteUsage()
  n30 = NamedRerouteDeclaration(None=MaterialFunctionCall(
    Brightness=ScalarParameter() {defaultValue=0.5, group=FN_Hair, parameterName=BaseColor_Fresnel_Brightness, sortPriority=1}
    Exponent=ScalarParameter() {defaultValue=3.5, group=FN_Hair, parameterName=Basecolor_Fresnel_Exponent, sortPriority=1}
    Vertex / Pixel Normal=PixelNormalWS()
   ) {materialFunction=&MF_Fresnel_FX}.Fresnel Inverted Double Sided) {name=fresnel, nodeColor=<rgba 0 1 0.2969 1>}
  n31 = Multiply(A=Multiply(A=@n29, B=@n30), B=ScalarParameter() {defaultValue=1, group=FN_Hair, parameterName=Emissive_Brightness, sortPriority=1})
  NamedRerouteDeclaration(None=Saturate(None=Multiply(A=@n31, B=@n28))) {name=simulate light transmission, nodeColor=<rgba 0 1 0.8594 1>}
  n32 = Add(
   A=Multiply(A=Multiply(A=@n29, B=@n30), B=ScalarParameter() {defaultValue=8, group=FN_Hair, parameterName=Fresnel_Brightness_Multiple, sortPriority=1})
   B=@n29
  )
  NamedRerouteDeclaration(None=Saturate(None=Saturate(None=@n32))) {name=basecolor with fresnel, nodeColor=<rgba 0.8984 0 1 1>}
  n33 = TextureSampleParameter2D() {
   constCoordinate=1
   group=FN_Hair
   parameterName=Paint_Hair_Texture
   samplerSource=SSM_Wrap_WorldGroupSettings
   samplerType=SAMPLERTYPE_Masks
   sortPriority=6
   texture=&T_BlackMask
  }
  NamedRerouteDeclaration(None=MaterialFunctionCall(In=@n33.G, Contrast=ScalarParameter() {group=FN_Hair, parameterName=Paint_Hair_Contrast}) {materialFunction=&CheapContrast}.Result) {name=InPut_Hair_Texture, nodeColor=<rgba 0.1016 0 1 1>}
  n34 = TextureSampleParameter2D(UVs=@n16.UVs) {
   group=FlyAwayHair
   parameterName=FlyAwayHairNormal
   samplerSource=SSM_Wrap_WorldGroupSettings
   samplerType=SAMPLERTYPE_Normal
   texture=/Game/Global/Materials/Character_Parent_2023/Hair_Parent_2023/Texture/T_FlyAwayHair_N
  }
  NamedRerouteDeclaration(None=MaterialFunctionCall(Normal=@n34.RGB, Flatness=Constant() {r=0.5}) {materialFunction=&FlattenNormal}.Result) {name=HairTextureNormal, nodeColor=<rgba 1 0.1875 0 1>}
  n35 = VertexNormalWS()
  n36 = NamedRerouteDeclaration(None=ComponentMask(None=TextureCoordinate()) {desc="black is tip ,white is root", g=true}) {name=RootTip_Gradient color, nodeColor=<rgba 0.2422 0 1 1>}
  n37 = Multiply(A=NamedRerouteDeclaration(None=OneMinus(None=@n36)) {name=Reverse RootTip, nodeColor=<rgba 1 0 0.4922 1>}, B=NamedRerouteUsage())
  n38 = ScalarParameter() {defaultValue=0.15, group=FlyAwayHair, parameterName=FlyAwayHair_TipWPO, sortPriority=2}
  n39 = NamedRerouteUsage()
  n40 = LinearInterpolate(
   A=ScalarParameter() {defaultValue=-0.05, group=FlyAwayHair, parameterName=FlyAwayHair_RootWPO, sortPriority=2}
   B=ScalarParameter() {defaultValue=0.25, group=FlyAwayHair, parameterName=FlyAwayHair_MIDWPO, sortPriority=2}
   Alpha=Add(A=Multiply(A=@n37) {constB=4}, B=LinearInterpolate(B=@n38, Alpha=@n39) {constB=0})
  ) {constA=-0.03, constB=0.6}
  NamedRerouteDeclaration(None=Add(A=@n11.worldPositionOffset, B=Multiply(A=@n35, B=@n40))) {name=OutPutWPO, nodeColor=<rgba 0 1 0.625 1>}
  NamedRerouteDeclaration(None=LinearInterpolate(
    A=ScalarParameter() {group=FlyAwayHair, parameterName=FlyAwayHair_ Darkness}
    B=ScalarParameter() {defaultValue=1, group=FlyAwayHair, parameterName=FlyAwayHair_ brightness}
    Alpha=@n17.R
   )) {name=FlyAwayHairtexture, nodeColor=<rgba 1 0 0.2109 1>}
 ]}
