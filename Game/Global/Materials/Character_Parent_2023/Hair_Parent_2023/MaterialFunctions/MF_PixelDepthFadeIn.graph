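# Material function for the parent hair material that provides a camera-distance fade mask and a Fresnel mask.
# The fade divides pixel depth by a distance parameter in the OverLay_Hair group and saturates, rising from zero near the camera to one at that distance.
# A separate MessdHairFresnel output takes a Fresnel on the world-space vertex normal, then scales, sharpens with a power and saturates it.
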MaterialFunction /Game/Global/Materials/Character_Parent_2023/Hair_Parent_2023/MaterialFunctions/MF_PixelDepthFadeIn
graph {nodes [
  Constant()
  n1 = PixelDepth()
  n2 = ScalarParameter() {defaultValue=120, group=OverLay_Hair, parameterName=PixDepthDiantance}
  FunctionOutput(None=Saturate(None=Divide(A=@n1.R, B=@n2) {constB=100}))
  n3 = Multiply(A=Fresnel(ExponentIn=Constant() {r=0.886}, Normal=VertexNormalWS()) {exponent=1}, B=Constant() {r=2.0763}) {constB=2}
  n4 = Constant() {r=1.5}
  FunctionOutput(None=Saturate(None=Power(Base=@n3, Exp=@n4) {constExponent=1.5})) {outputName=MessdHairFresnel}
 ]}
