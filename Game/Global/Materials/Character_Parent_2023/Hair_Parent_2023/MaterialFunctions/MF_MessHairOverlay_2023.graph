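# Material function for a messy-hair overlay on character hair, rewriting incoming material attributes for color, normal, roughness, opacity and vertex offset.
# A packed mask combines two strand layers, one parallaxed with bump offset, into dithered temporal-AA and plain masks blended by pixel depth.
# Shadow Replace zeroes the mask in shadow passes.
# World-aligned noise sampled at pre-skinned position drives mask-weighted displacement along normals, with depth fade also adjusting displacement and color blend.
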MaterialFunction /Game/Global/Materials/Character_Parent_2023/Hair_Parent_2023/MaterialFunctions/MF_MessHairOverlay_2023
graph {nodes [
  n1 = Multiply(A=VertexNormalWS(), B=TwoSidedSign())
  n2 = SetMaterialAttributes!(
   MaterialAttributes=NamedRerouteUsage()
   BaseColor=NamedRerouteUsage()
   OpacityMask=NamedRerouteUsage()
   WorldPositionOffset=NamedRerouteUsage()
   Opacity=LinearInterpolate(A=Constant() {r=1}, Alpha=TwoSidedSign())
   SubsurfaceColor=Constant()
   Roughness=ScalarParameter() {defaultValue=0.5, group=OverLay_Hair, parameterName=MessHairRoughness}
   Normal=MaterialFunctionCall(
    BaseNormal=Transform(None=@n1) {transformSourceType=TRANSFORMSOURCE_World, transformType=TRANSFORM_Tangent}
    AdditionalNormal=NamedRerouteUsage()
   ) {materialFunction=&BlendAngleCorrectedNormals}.Result
   Metallic=Constant()
  ) {attributeSetTypes [
    #1
    #2
    #3
    #4
    #5
    #6
    #7
    #8
   ]}
  FunctionOutput(None=@n2)
  n3 = GetMaterialAttributes(None=FunctionInput() {inputName=MA, inputType=FunctionInput_MaterialAttributes}) {attributeGetTypes [#1, #7]}
  NamedRerouteDeclaration(None=@n3.BaseColor) {name=InputBaseColor, nodeColor=<rgba 1 0 0.6328 1>}
  NamedRerouteDeclaration(None=@n3.Normal) {name=InputNormal, nodeColor=<rgba 0.3125 1 0 1>}
  n4 = TextureSampleParameter2D(UVs=MaterialFunctionCall!(UV Coord=TextureCoordinate(), Height=Constant() {r=0.05}, ReferencePlane=Constant(), HeightRatio=Constant() {r=0.1}, Normal) {materialFunction=/Game/Global/MaterialFunctions/MF_BumpOffset_Accurate}.Result) {
   group=OverLay_Hair
   parameterName=MessHairMask
   samplerSource=SSM_Wrap_WorldGroupSettings
   samplerType=SAMPLERTYPE_Masks
   texture=&T_BlackMask
  }
  n5 = NamedRerouteDeclaration(None=Power(Base=@n4.G, Exp=ScalarParameter() {defaultValue=3.07, group=OverLay_Hair, parameterName=DitherMess hair intensity G}) {constExponent=1.8}) {name=MessHairMaskB_2, nodeColor=<rgba 1 0 0.9141 1>}
  n6 = TextureSampleParameter2D(UVs=TextureCoordinate()) {
   group=OverLay_Hair
   parameterName=MessHairMask
   samplerSource=SSM_Wrap_WorldGroupSettings
   samplerType=SAMPLERTYPE_Masks
   texture=&T_BlackMask
  }
  n7 = NamedRerouteDeclaration(None=Power(Base=@n6.R, Exp=ScalarParameter() {defaultValue=3, group=OverLay_Hair, parameterName=DitherMess hair intensityR}) {constExponent=3}) {name=MessHairMaskA_2, nodeColor=<rgba 0 1 0.1094 1>}
  n8 = Multiply(
   A=Multiply(A=Max(A=@n5, B=@n7), B=@n4.B)
   B=ScalarParameter() {defaultValue=10, group=OverLay_Hair, parameterName=DitherMessHairMask_Brightness}
  ) {constB=10}
  n9 = ScalarParameter() {defaultValue=0.27, group=OverLay_Hair, parameterName=DitherMessHairMask_Contrast}
  n10 = LinearInterpolate(A=Constant() {r=-0.1}, B=Constant() {r=2}, Alpha=Multiply(A=Power(Base=@n8, Exp=@n9), B=NamedRerouteUsage())) {constA=0.5}
  n11 = Saturate(None=Divide(A=PixelDepth().R, B=Constant() {r=30}) {constB=100})
  n12 = NamedRerouteDeclaration(None=MaterialFunctionCall(Alpha Threshold=Saturate(None=@n10), Random=LinearInterpolate(Alpha=@n11) {constA=1, constB=0}) {materialFunction=&DitherTemporalAA}.Result) {name=DitherOpacitymask, nodeColor=<rgba 0.9922 0 1 1>}
  n13 = NamedRerouteUsage()
  n14 = Power(Base=MaterialFunctionCall() {materialFunction=&/MF_PixelDepthFadeIn}.Result) {constExponent=3}
  NamedRerouteDeclaration(None=ShadowReplace(Default=LinearInterpolate(A=@n12, B=@n13, Alpha=@n14), Shadow=Constant())) {name=Opacitymask_1, nodeColor=<rgba 0.0078 0 1 1>}
  n15 = ScalarParameter() {group=OverLay_Hair, parameterName=BaseColorBlendClose, sortPriority=50}
  n16 = ScalarParameter() {defaultValue=0.25, group=OverLay_Hair, parameterName=BaseColorBlendFar, sortPriority=50}
  n17 = MaterialFunctionCall() {materialFunction=&MF_PixelDepthFadeIn}
  NamedRerouteDeclaration(None=LinearInterpolate(
    A=VectorParameter() {group=OverLay_Hair, parameterName=MessHairColorA, defaultValue=<rgba 0.4115 0.2299 0.1564 1>}.RGB
    B=NamedRerouteUsage()
    Alpha=LinearInterpolate(A=@n15, B=@n16, Alpha=@n17.Result)
   )) {name=BaseColor_1, nodeColor=<rgba 0.3125 1 0 1>}
  NamedRerouteDeclaration(None=@n3.MaterialAttributes) {name=InputMA_1, nodeColor=<rgba 1 0 0.2344 1>}
  n18 = TextureSampleParameter2D(UVs=MaterialFunctionCall(UV Coord=TextureCoordinate(), Height=Constant() {r=0.05}, ReferencePlane=Constant(), HeightRatio=Constant() {r=0.1}) {materialFunction=&MF_BumpOffset_Accurate}.Result) {
   desc=R-MesssHairLayer1 G-MessHairLayLayer2 B -MesshairMask
   group=GnatGala_MessHair
   parameterName=MessHairMask
   samplerSource=SSM_Wrap_WorldGroupSettings
   samplerType=SAMPLERTYPE_Masks
   texture=&T_BlackMask
  }
  n19 = Constant() {r=3.5087}
  n20 = TextureSampleParameter2D(UVs=TextureCoordinate()) {
   group=OverLay_Hair
   parameterName=MessHairMask
   samplerSource=SSM_Wrap_WorldGroupSettings
   samplerType=SAMPLERTYPE_Masks
   texture=&T_BlackMask
  }
  n21 = Constant() {r=2.2404}
  n22 = Max(
   A=NamedRerouteDeclaration(None=Power(Base=@n18.G, Exp=@n19) {constExponent=1.8}) {name=MessHairMaskB_1, nodeColor=<rgba 1 0 0.9141 1>}
   B=NamedRerouteDeclaration(None=Power(Base=@n20.R, Exp=@n21) {constExponent=3}) {name=MessHairMaskA_1, nodeColor=<rgba 0 1 0.1094 1>}
  )
  n23 = Power(Base=Multiply(A=Multiply(A=@n22, B=@n18.B), B=Constant() {r=10}) {constB=10}, Exp=Constant() {r=6.9479})
  n24 = NamedRerouteUsage()
  NamedRerouteDeclaration(None=Saturate(None=Multiply(A=@n23, B=@n24))) {name=BaseOpacitymask, nodeColor=<rgba 0.0078 0 1 1>}
  n25 = TextureObjectParameter() {
   group=OverLay_Hair
   parameterName=WPONoise
   samplerType=SAMPLERTYPE_Grayscale
   texture=/Game/Global/Textures/Default/Noise/LowResBlurredNoiseNormalized
  }
  n26 = ScalarParameter() {defaultValue=21, group=OverLay_Hair, parameterName=WPONoiseTextureSize}
  n27 = Add(
   A=FunctionInput() {bUsePreviewValueAsDefault=true, inputName=PreSkinLOC, sortPriority=3}
   B=VectorParameter() {group=OverLay_Hair, parameterName=WPONoiseOffset, defaultValue=<rgba -1.3222 4.288 -1.4387 1>}.RGB
  )
  NamedRerouteDeclaration(None=LinearInterpolate(
    A=ScalarParameter() {defaultValue=0.02, group=OverLay_Hair, parameterName=WPONoiseMin}
    B=ScalarParameter() {defaultValue=0.9139, group=OverLay_Hair, parameterName=WPONoiseMax}
    Alpha=MaterialFunctionCall(TextureObject=@n25, TextureSize=@n26, WorldPosition=@n27) {materialFunction=&WorldAlignedTexture}.XYZ Texture
   )) {name=Noise_1, nodeColor=<rgba 0.0547 0 1 1>}
  n28 = VertexNormalWS()
  n29 = NamedRerouteUsage()
  n30 = TextureSampleParameter2D() {
   group=GnatGala_MessHair
   parameterName=MessHairMask
   samplerSource=SSM_Wrap_WorldGroupSettings
   samplerType=SAMPLERTYPE_Masks
   texture=&T_BlackMask
  }
  n31 = Multiply(
   A=LinearInterpolate(
    A=ScalarParameter() {defaultValue=0.05, group=OverLay_Hair, parameterName=WPO_Min}
    B=ScalarParameter() {defaultValue=1, group=OverLay_Hair, parameterName=WPO_Max}
    Alpha=Power(Base=@n29)
   )
   B=@n30.B
  )
  n32 = Power(Base=MaterialFunctionCall() {materialFunction=&MF_PixelDepthFadeIn}.Result)
  NamedRerouteDeclaration(None=Multiply(A=Multiply(A=@n28, B=@n31), B=LinearInterpolate(Alpha=@n32) {constA=1.3, constB=0.4})) {name=WPO_1, nodeColor=<rgba 1 0.3281 0 1>}
 ]}
